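# Masked parent material for the Oldmans vehicle wheels, usable on skeletal meshes and with mesh deformers.
# MF_WheelParent_Simple supplies the base attributes, then the opacity mask is overridden by a lerp from 1 to the Alpha_Mask scalar, weighted by inverted rim texture alpha.
# At the default Alpha_Mask of 1 nothing is cut out; lower values reduce the mask where texture alpha is low.
# The result passes through vehicle damage, then de-rez transition functions.
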Material /VehicleCosmetics/Wheels/Oldmans/New/Materials/M_WheelParent_Oldmans
graph {
 properties {
  bAutomaticallySetUsageInEditor=false
  bUseMaterialAttributes=true
  bUsedWithMeshDeformer=true
  bUsedWithSkeletalMesh=true
  blendMode=BLEND_Masked
  pixelDepthOffsetMode=PDOM_Legacy
  previewMesh=/VehicleCosmetics/Wheels/Oldmans/New/Meshes/NEW/SK_Wheel_Oldmans
  usedShadingModels=Default Lit
 }
 nodes [
  n1 = GetMaterialAttributes(None=MaterialFunctionCall() {materialFunction=&MF_WheelParent_Simple}.MaterialAttributes) {attributeGetTypes [#1]}
  n2 = LinearInterpolate(
   B=ScalarParameter() {defaultValue=1, parameterName=Alpha_Mask}
   Alpha=OneMinus(None=TextureSampleParameter2D() {parameterName=Alpha_Mask_BCA, texture=/VehicleCosmetics/Wheels/Oldmans/New/Textures/T_Oldmans_Rim_BC_LOD}.A)
  ) {constA=1}
  MaterialFunctionCall(In=MaterialFunctionCall(Material=SetMaterialAttributes!(MaterialAttributes=@n1.MaterialAttributes, Opacity Mask=@n2) {attributeSetTypes [#2]}) {materialFunction=&MF_BaseVehicle_Damage}.Result) {materialFunction=&MF_VehicleDeRez_Transition}
 ]
}
