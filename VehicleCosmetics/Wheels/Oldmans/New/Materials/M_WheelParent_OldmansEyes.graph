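# Parent material for the Oldmans vehicle wheel's eyes, animating eye textures so they dart around on the wheel mesh.
# Three UV channels each get layered, clamped Noise1D time signals that drive separate X and Y panners, with differing offsets and rates to vary the motion.
# The shifted color is multiplied by an AO texture and summed across channels, as are the normals.
# The result passes through shared vehicle damage and de-rez transition functions.
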Material /VehicleCosmetics/Wheels/Oldmans/New/Materials/M_WheelParent_OldmansEyes
graph {
 properties {
  bAutomaticallySetUsageInEditor=false
  bUseMaterialAttributes=true
  bUsedWithMeshDeformer=true
  bUsedWithSkeletalMesh=true
  previewMesh=/VehicleCosmetics/Wheels/Oldmans/New/Meshes/SK_Wheel_Oldmans
  usedShadingModels=Default Lit
 }
 nodes [
  n1 = TextureCoordinate()
  n2 = TextureSample(UVs=@n1) {samplerType=SAMPLERTYPE_LinearColor, texture=/VehicleCosmetics/Wheels/Oldmans/New/Textures/oldmans_eyes_ao}
  n3 = Constant() {r=-0.1}
  n4 = Time()
  n5 = Add(A=Add(A=@n3, B=@n3), B=@n4)
  n6 = Constant() {r=0.4}
  n7 = Constant() {r=40}
  n8 = Multiply(A=MaterialFunctionCall!(In=Multiply(A=@n5, B=@n6)) {materialFunction=/VehicleCosmetics/Wheels/NewtonBall/Materials/Noise1D}.Result, B=@n7)
  n9 = Constant() {r=2.3}
  n10 = Constant() {r=0.5}
  n11 = Time()
  n12 = Constant() {r=20}
  n13 = Multiply(A=MaterialFunctionCall(In=Multiply(A=@n10, B=@n11)) {materialFunction=&Noise1D}.Result, B=@n12)
  n14 = Constant() {r=0.5}
  n15 = Add(A=Multiply(A=Clamp(None=@n8) {minDefault=-1}, B=@n9), B=Multiply(A=Clamp(None=@n13) {minDefault=-1}, B=@n14))
  n16 = Constant() {r=-0.2}
  n17 = Multiply(A=MaterialFunctionCall(In=Multiply(A=@n5, B=@n16)) {materialFunction=&Noise1D}.Result, B=@n7)
  n18 = Constant() {r=-0.6}
  n19 = Multiply(A=MaterialFunctionCall(In=Multiply(A=@n11, B=@n18)) {materialFunction=&Noise1D}.Result, B=@n12)
  n20 = Add(A=Multiply(A=Clamp(None=@n17) {minDefault=-1}, B=@n9), B=Multiply(A=Clamp(None=@n19) {minDefault=-1}, B=@n14))
  n21 = AppendVector(
   A=ComponentMask(None=Panner(Coordinate=@n1, Time=@n15) {speedX=0.04, speedY=0.02}) {r=true}
   B=ComponentMask(None=Panner(Coordinate=@n1, Time=@n20) {speedY=0.04}) {g=true}
  )
  n22 = Multiply(
   A=@n2.RGB
   B=TextureSampleParameter2D(UVs=@n21) {parameterName=Oldmans_Eyes, texture=/VehicleCosmetics/Wheels/Oldmans/New/Textures/T_Oldmans_Eyes_BC}.RGB
  )
  n23 = TextureCoordinate() {coordinateIndex=1}
  n24 = TextureSample(UVs=@n23) {samplerType=SAMPLERTYPE_LinearColor, texture=&oldmans_eyes_ao}
  n25 = Add(A=@n3, B=@n4)
  n26 = Constant() {r=40}
  n27 = Multiply(A=MaterialFunctionCall(In=Multiply(A=@n25, B=@n6)) {materialFunction=&Noise1D}.Result, B=@n26)
  n28 = Constant() {r=2.3}
  n29 = Constant() {r=0.5}
  n30 = Time()
  n31 = Constant() {r=20}
  n32 = Multiply(A=MaterialFunctionCall(In=Multiply(A=@n29, B=@n30)) {materialFunction=&Noise1D}.Result, B=@n31)
  n33 = Constant() {r=0.5}
  n34 = Add(A=Multiply(A=Clamp(None=@n27) {minDefault=-1}, B=@n28), B=Multiply(A=Clamp(None=@n32) {minDefault=-1}, B=@n33))
  n35 = Multiply(A=MaterialFunctionCall(In=Multiply(A=@n25, B=@n16)) {materialFunction=&Noise1D}.Result, B=@n26)
  n36 = Constant() {r=1}
  n37 = Multiply(A=MaterialFunctionCall(In=Multiply(A=@n30, B=@n36)) {materialFunction=&Noise1D}.Result, B=@n31)
  n38 = Add(A=Multiply(A=Clamp(None=@n35) {minDefault=-1}, B=@n28), B=Multiply(A=Clamp(None=@n37) {minDefault=-1}, B=@n33))
  n39 = AppendVector(
   A=ComponentMask(None=Panner(Coordinate=@n23, Time=@n34) {speedX=0.04, speedY=0.02}) {r=true}
   B=ComponentMask(None=Panner(Coordinate=@n23, Time=@n38) {speedY=0.04}) {g=true}
  )
  n40 = Multiply(
   A=@n24.RGB
   B=TextureSampleParameter2D(UVs=@n39) {parameterName=Oldmans_Eyes, texture=&T_Oldmans_Eyes_BC}.RGB
  )
  n41 = TextureCoordinate() {coordinateIndex=2}
  n42 = TextureSample(UVs=@n41) {samplerType=SAMPLERTYPE_LinearColor, texture=&oldmans_eyes_ao}
  n43 = Constant() {r=40}
  n44 = Multiply(A=@n43, B=MaterialFunctionCall(In=Multiply(A=@n4, B=@n6)) {materialFunction=&Noise1D}.Result)
  n45 = Constant() {r=2.3}
  n46 = Constant() {r=2}
  n47 = Subtract(A=Constant(), B=@n46)
  n48 = AppendVector(A=@n47, B=@n47)
  n49 = Constant() {r=1}
  n50 = Time()
  n51 = Multiply(A=Constant() {r=1.5}, B=@n50)
  n52 = Clamp(None=LinearInterpolate(
    A=AppendVector(A=@n48, B=@n47)
    B=Add(A=@n49, B=@n46)
    Alpha=MaterialFunctionCall(In=@n51) {materialFunction=&Noise1D}.Result
   )) {minDefault=-1}
  n53 = Constant() {r=0.4}
  n54 = Add(A=Multiply(A=Clamp(None=@n44) {minDefault=-1}, B=@n45), B=Multiply(A=ComponentMask(None=@n52) {r=true}, B=@n53))
  n55 = Multiply(A=@n43, B=MaterialFunctionCall(In=Multiply(A=@n16, B=@n4)) {materialFunction=&Noise1D}.Result)
  n56 = Subtract(A=Constant(), B=@n46)
  n57 = AppendVector(A=@n56, B=@n56)
  n58 = Multiply(A=@n50, B=Constant() {r=1})
  n59 = Constant() {r=1}
  n60 = Clamp(None=LinearInterpolate(
    A=AppendVector(A=@n57, B=@n56)
    B=Add(A=@n59, B=@n46)
    Alpha=MaterialFunctionCall(In=@n58) {materialFunction=&Noise1D}.Result
   )) {minDefault=-1}
  n61 = Add(A=Multiply(A=Clamp(None=@n55) {minDefault=-1}, B=@n45), B=Multiply(A=ComponentMask(None=@n60) {r=true}, B=@n53))
  n62 = AppendVector(
   A=ComponentMask(None=Panner(Coordinate=@n41, Time=@n54) {speedX=0.04, speedY=0.02}) {r=true}
   B=ComponentMask(None=Panner(Coordinate=@n41, Time=@n61) {speedY=0.04}) {g=true}
  )
  n63 = TextureSampleParameter2D(UVs=@n62) {parameterName=Oldmans_Eyes, texture=&T_Oldmans_Eyes_BC}
  n64 = Multiply(A=Add(A=Add(A=@n2.RGB, B=@n24.RGB), B=@n42.RGB), B=ScalarParameter() {defaultValue=1, parameterName=Roughness})
  n65 = ScalarParameter() {defaultValue=0.5, parameterName=Eyes_Roughness}
  n66 = TextureSampleParameter2D(UVs=@n21) {parameterName=Oldmans_Eyes_N, samplerType=SAMPLERTYPE_Normal, texture=/VehicleCosmetics/Wheels/Oldmans/New/Textures/T_Oldmans_Eyes_N}
  n67 = TextureSampleParameter2D(UVs=@n39) {parameterName=Oldmans_Eyes_N, samplerType=SAMPLERTYPE_Normal, texture=&T_Oldmans_Eyes_N}
  n68 = TextureSampleParameter2D(UVs=@n62) {parameterName=Oldmans_Eyes_N, samplerType=SAMPLERTYPE_Normal, texture=&T_Oldmans_Eyes_N}
  n69 = SetMaterialAttributes!(
   MaterialAttributes
   Base Color=Add(A=Add(A=@n22, B=@n40), B=Multiply(A=@n42.RGB, B=@n63.RGB))
   Metallic=Constant()
   Roughness=OneMinus(None=LinearInterpolate(A=@n64, B=@n65))
   Normal=Add(A=@n66.RGB, B=Add(A=@n67.RGB, B=@n68.RGB))
  ) {attributeSetTypes [
    #1
    #2
    #3
    #4
   ]}
  MaterialFunctionCall(In=MaterialFunctionCall(Material=@n69) {materialFunction=&MF_BaseVehicle_Damage}.Result) {materialFunction=&MF_VehicleDeRez_Transition}
 ]
}
