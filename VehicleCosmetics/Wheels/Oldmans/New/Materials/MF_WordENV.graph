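# Material function that fakes an environment reflection from a grass cubemap, apparently for wheel cosmetics.
# The reflection vector samples the cube directly for the LocalSpace output and through a default vector Transform for WorldSpace_1.
# Both samples pass through Desaturate, then blend toward a tinted multiply by TintAmount.
# Tint previews with a fixed-UV sample of the Team1_Color_LookUp gradient parameter.
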MaterialFunction /VehicleCosmetics/Wheels/Oldmans/New/Materials/MF_WordENV
graph {nodes [
  n1 = ReflectionVectorWS()
  n2 = TextureObject() {samplerType=SAMPLERTYPE_LinearColor, texture=/VehicleCosmetics/Wheels/DuelingDragonsSE/Materials/TC_DuelingDragonsSE_GrassCube}
  n3 = FunctionInput(Preview=Constant()) {inputName=Desaturate, inputType=FunctionInput_Scalar}
  n4 = Desaturation(None=TextureSample(UVs=@n1, Tex=@n2).RGB, Fraction=@n3)
  n5 = TextureSampleParameter2D(UVs=Constant2Vector() {g=0.5, r=0.67}) {parameterName=Team1_Color_LookUp, texture=/VehicleCosmetics/DriftTrail/DriftTrail_Techno/Textures/T_VCVFX_Drift_Color_Gradient_Green_D}
  n6 = FunctionInput(Preview=@n5.RGBA) {inputName=Tint}
  n7 = FunctionInput(Preview=Constant() {r=1}) {inputName=TintAmount, inputType=FunctionInput_Scalar}
  FunctionOutput(None=LinearInterpolate(A=@n4, B=Multiply(A=@n4, B=@n6), Alpha=@n7)) {outputName=LocalSpace}
  n8 = Desaturation(None=TextureSample(UVs=Transform(None=@n1), Tex=@n2).RGBA, Fraction=@n3)
  FunctionOutput(None=LinearInterpolate(A=@n8, B=Multiply(A=@n8, B=@n6), Alpha=@n7)) {outputName=WorldSpace_1, sortPriority=1}
 ]}
